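; Gameplay ability for BuzzCool that registers its avatar pawn with the BuzzCool manager as the pawn with the donut.
; On activation, the authority finds tagged manager actors, caches the last one as ManagingActor and passes it the avatar's valid Fort player pawn.
; When the ability ends, it tells the cached manager to clear that pawn, with no authority or validity check in this graph.
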(event Ability|EventActivateAbility
  (Ability|HasAuthority
    (:True
      (bind _outactors (Actor|GetAllActorsOfClasswithTag "/BuzzCool/Gameplay/Manager/BGA_BuzzCool_Manager.BGA_BuzzCool_Manager_C" "BuzzCoolManager"))
      (Utilities|IsValid (Utilities|Array|Get(acopy) _outactors (Utilities|Array|LastIndex _outactors))
        (:Is%20Valid
          (bind _output_get (Variables|Default|SetManagingActor (Utilities|Array|Get(acopy) _outactors (Utilities|Array|LastIndex _outactors))))
          (bind _outfortplayerpawn (CallFunction|GetAvatarasValidFortPlayerPawn)
            (:ValidFortPlayerPawn
              (Class|BGABuzzCoolManager|SetPawnwithDonut _output_get _outfortplayerpawn :!class /BuzzCool/Gameplay/Manager/BGA_BuzzCool_Manager.BGA_BuzzCool_Manager_C))))))))

(event Ability|EventOnEndAbility
  (params (bWasCancelled :type Boolean))
  (Class|BGABuzzCoolManager|ClearPawnwithDonut (Variables|Default|GetManagingActor)))
